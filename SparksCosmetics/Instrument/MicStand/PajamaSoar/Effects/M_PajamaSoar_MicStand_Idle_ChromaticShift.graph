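# Translucent, unlit, two-sided Niagara effect material for the PajamaSoar mic stand's idle chromatic shift.
# Emissive blends two color pairs along a masked UV axis, swaps them between front and back faces, and tints by particle color.
# Opacity multiplies optional sphere and edge masks, transition fade, particle alpha, depth fade and view-angle fade.
# Optional SplineThicken offsets vertices, with the dynamic base width feeding both base and tip.
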Material /SparksCosmetics/Instrument/MicStand/PajamaSoar/Effects/M_PajamaSoar_MicStand_Idle_ChromaticShift
graph {
 properties {
  bAutomaticallySetUsageInEditor=false
  bUsedWithNiagaraMeshParticles=true
  bUsedWithNiagaraRibbons=true
  bUsedWithNiagaraSprites=true
  bUsedWithSplineMeshes=true
  blendMode=BLEND_Translucent
  shadingModel=MSM_Unlit
  twoSided=true
  usedShadingModels=Unlit
 }
 nodes [
  n1 = DynamicParameter() {paramNames [spline width base, spline width tip, Param3, Param4]}
  LinearInterpolate(
   B=MaterialFunctionCall(WidthBase=@n1.spline width base, WidthTip=@n1.spline width base) {materialFunction=&SplineThicken}.WorldPositionOffset
   Alpha=ScalarParameter() {defaultValue=1, desc="1=EnableSplineThicken", parameterName=UseSplineThicken, sortPriority=17}
  ) {constAlpha=1, constB=0}
  n2 = StaticComponentMaskParameter(None=TextureCoordinate()) {defaultR=true, parameterName=UV Coordinate Color Split mask, sortPriority=7}
  n3 = OneMinus(None=Sine(None=Multiply(A=@n2) {constB=0.5}))
  n4 = Power(Base=@n3, Exp=ScalarParameter() {defaultValue=1, parameterName=Color Transition Power})
  n5 = LinearInterpolate(
   A=VectorParameter() {parameterName=Color 2B, sortPriority=13, defaultValue=<rgba 0 0.0257 0.1823 1>}.RGB
   B=VectorParameter() {parameterName=Color 2, sortPriority=12, defaultValue=<rgba 0 0.1875 0.173 1>}.RGB
   Alpha=@n4
  )
  n6 = LinearInterpolate(
   A=VectorParameter() {parameterName=Color 1, sortPriority=10, defaultValue=<rgba 0.1146 0 0.002 1>}.RGB
   B=VectorParameter() {parameterName=Color 1b, sortPriority=11, defaultValue=<rgba 0.3438 0.0271 0.0233 1>}.RGB
   Alpha=@n4
  )
  n7 = MaterialFunctionCall!(x=@n2, size) {materialFunction=/Game/Effects/Fort_Effects/MaterialFunctions/MF_Fill}
  n8 = LinearInterpolate(A=@n5, B=@n6, Alpha=@n7.Result)
  n9 = LinearInterpolate(A=@n6, B=@n5, Alpha=@n7.Result)
  n10 = Saturate(None=TwoSidedSign())
  n11 = ParticleColor()
  MaterialFunctionCall(Emissive=Multiply(A=LinearInterpolate(A=@n8, B=@n9, Alpha=@n10), B=@n11.RGB)) {materialFunction=&MF_FX_EmissiveAdjustments}
  n12 = TextureCoordinate()
  n13 = Constant() {r=0.5}
  n14 = ScalarParameter() {defaultValue=0.6, parameterName=Sphere Mask Radius, sortPriority=2}
  n15 = ScalarParameter() {defaultValue=0.5, parameterName=Sphere Mask Hardness, sortPriority=3}
  n16 = Sine(None=Multiply(A=MaterialFunctionCall() {materialFunction=&LinearGradient}.VGradient) {constB=0.5})
  n17 = ScalarParameter() {defaultValue=1, parameterName=V Fade Power}
  n18 = SmoothStep(
   Max=ScalarParameter() {defaultValue=0.6, parameterName=V Fade, sortPriority=6}
   Value=Saturate(None=Power(Base=@n16, Exp=@n17) {constExponent=1})
  ) {constMax=0.5}
  n19 = Sine(None=Multiply(A=MaterialFunctionCall() {materialFunction=&LinearGradient}.UGradient) {constB=0.5})
  n20 = ScalarParameter() {defaultValue=1, parameterName=U Fade Power}
  n21 = SmoothStep(
   Max=ScalarParameter() {defaultValue=0.8, parameterName=U Fade, sortPriority=5}
   Value=Saturate(None=Power(Base=@n19, Exp=@n20) {constExponent=1})
  ) {constMax=0.5}
  n22 = Multiply(
   A=LinearInterpolate(
    B=SphereMask(A=@n12, B=@n13, Radius=@n14, Hardness=@n15)
    Alpha=ScalarParameter() {defaultValue=0.5, parameterName=Use Sphere Mask, sortPriority=1}
   ) {constA=1}
   B=LinearInterpolate(B=Multiply(A=@n18, B=@n21), Alpha=ScalarParameter() {defaultValue=1, parameterName=Use Edge Mask, sortPriority=4}) {constA=1}
  )
  n23 = ScalarParameter() {defaultValue=1, parameterName=transition fade power, sortPriority=8}
  n24 = SmoothStep(
   Max=ScalarParameter() {defaultValue=0.05, parameterName=transition fade, sortPriority=9}
   Value=Saturate(None=Power(Base=@n3, Exp=@n23) {constExponent=1})
  ) {constMax=0.5}
  n25 = DepthFade(FadeDistance=ScalarParameter() {defaultValue=10, parameterName=DepthFade, sortPriority=14}) {fadeDistanceDefault=15}
  n26 = ScalarParameter() {defaultValue=1, parameterName=AngleFade Strength, sortPriority=15}
  Multiply(
   A=Multiply(A=Multiply(A=@n22, B=@n24), B=Multiply(A=@n11.A, B=@n25))
   B=Max(
    A=MaterialFunctionCall("FadeSensitivity (optional)"=@n26) {materialFunction=&FadeBasedOnViewAngle}.Result
    B=ScalarParameter() {defaultValue=0.15, parameterName=Max Angle Fade, sortPriority=16}
   ) {constB=0.15}
  )
 ]
}
