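# Idle glow effect for the PajamaSoar mic stand cosmetic, built from two emitters, one for the mic and one for the stand.
# Each bursts persistent sprite and mesh particles, split by spawn group and pinned to skeletal mesh sockets.
# IncludeMicFX gates mic spawns, and IsBackpack switches mic offsets.
# System mesh-scale values drive sizing and material parameters, while time of day and front-end flags adjust color and alpha.
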NiagaraSystem /SparksCosmetics/Instrument/MicStand/PajamaSoar/Effects/NS_MicStand_PajamaSoar_Idle_FX
dependencies {
 usedDataInterfaces [
  &NiagaraDataInterfaceArrayInt32
  &NiagaraDataInterfaceCurve
  &NiagaraDataInterfaceCamera
  &NiagaraDataInterfaceArrayFloat
  &NiagaraDataInterfaceArrayFloat3
  &NiagaraDataInterfaceSkeletalMesh
  &NiagaraDataInterfaceColorCurve
 ]
 usedDynamicInputs [
  &MakeFloatFromVector
  &Add_Float
  &ColorFromCurve
 ]
 usedModules [
  &self:SystemSpawnScript.SystemScriptSource.SystemScriptGraph.NiagaraNodeAssignment_0.SetVariables_#1
  &SystemState
  &self:SystemSpawnScript.SystemScriptSource.SystemScriptGraph.NiagaraNodeAssignment_3.SetVariables_#2
  &self:SystemSpawnScript.SystemScriptSource.SystemScriptGraph.NiagaraNodeAssignment_2.SetVariables_#3
 ]
}
emitters {
 MicFX_Combined {
  emitterUpdateScript [
   &EmitterState(
    Inactive Response=::NewEnumerator1
    Life Cycle Mode=::NewEnumerator1
    Loop Behavior=::NewEnumerator1
    Loop Duration Mode=::NewEnumerator1
    UseLoopDelay=false
   )
   &SpawnBurst_Instantaneous(Spawn Probability=$User.IncludeMicFX, Use Spawn Probability=true, UseLoopCountLimit=true)
   &SpawnBurst_Instantaneous(Spawn Group=1, Spawn Probability=$User.IncludeMicFX, Use Spawn Probability=true, UseLoopCountLimit=true) {module=SpawnBurst_Instantaneous001}
   &SpawnBurst_Instantaneous(Spawn Group=2, Spawn Probability=$User.IncludeMicFX, Use Spawn Probability=true, UseLoopCountLimit=true) {module=SpawnBurst_Instantaneous002}
   &SpawnBurst_Instantaneous(Spawn Group=3, Spawn Probability=$User.IncludeMicFX, Use Spawn Probability=true, UseLoopCountLimit=true) {module=SpawnBurst_Instantaneous003}
  ]
  particleSpawnScript [
   /SetVariables(Particles.VisibilityTag=&SpawnGroup()) {isSetParameters=true}
   &InitializeParticle(
    Color=<rgba 1 1 1 0.75>
    Mesh Scale=&MaskVectorBySpawnGroup(Masked Vector=<1.2 1.2 1.2>, Passthrough Vector=<1 1 1>, Spawn Group Mask=2)
    Mesh Scale Mode=::NewEnumerator4
    Sprite Size=&MaskVector2dBySpawnGroup(Masked Vector2D=<11.5 17.5>, Passthrough Vector2D=<15.75 10>, Spawn Group Mask=0)
    Sprite Size Mode=::NewEnumerator4
    Sprite UV Mode=::NewEnumerator4
    UsePositionOffset=false
    Write Material Random=true
   )
   &CameraOffset(Camera Offset Amount=&SelectFloatFromArray(
     Direct Array Index=&SpawnGroup()
     Float Selection Array=DataInterface() {propertyValues {FloatData [-3,-3,5,-2]}}
    ))
   &InitialMeshOrientation(
    Mesh Orientation Mode=::NewEnumerator2
    Rotation=&MaskVectorBySpawnGroup(Masked Vector=<0 0 0.25>, Passthrough Vector=<0 0 0>, Spawn Group Mask=2)
    Use Rotation Vector=true
   )
  ]
  particleUpdateScript [
   &ParticleState(Kill Particles When Lifetime Has Elapsed=false, Let Infinitely Lived Particles Die When Emitter Deactivates=true)
   &SkeletalMeshLocation(
    MaskSpawnGroup=false
    Mesh Sampling Type=::NewEnumerator5
    Sampled Position Offset=&MaskVectorBySpawnGroup(
     Masked Vector=&MakeCustomVectorFromBool(
      Bool=&SetBoolByFloatComparison(A=$User.IsBackpack, B=0)
      True Vector=&Multiply_VectorByFloat(Vector=<8.3 -12.5 -1.5>, Float=$System.MeshScaleRemap)
      False Vector=&Multiply_VectorByFloat(Vector=<8.3 -12.5 0>, Float=$System.MeshScaleRemap)
     )
     Passthrough Vector=&MakeCustomVectorFromBool(
      Bool=&SetBoolByFloatComparison(A=$User.IsBackpack, B=0)
      True Vector=&Multiply_VectorByFloat(Vector=<14.25 -12 -2.25>, Float=$System.MeshScaleRemap)
      False Vector=&Multiply_VectorByFloat(Vector=<14.5 -12 0>, Float=$System.MeshScaleRemap)
     )
    )
    Skeletal Mesh=$System.Skeletal Mesh
    Socket Sampling Mode=::NewEnumerator5
   ) {enabled=false}
   &ScaleSpriteSize(Scale Sprite Size Mode=::NewEnumerator0, Uniform Scale Factor=&Multiply_Float(A=1, B=$System.MeshScaleRemap))
   &SkeletalMeshLocation(
    Bone / Socket Index=&SelectIntFromArray(Direct Array Index=&SpawnGroup(), Int Selection Array=DataInterface() {propertyValues {IntData [0,0,0,3]}})
    MaskSpawnGroup=false
    Mesh Sampling Type=::NewEnumerator5
    Sampled Position Offset=&Multiply_VectorByFloat(
     Vector=&MakeCustomVectorFromBool(
      Bool=&SetBoolByFloatComparison(A=$User.IsBackpack, B=0)
      True Vector=&SelectVectorFromArray(
       Direct Array Index=&SpawnGroup()
       Vector Selection Array=DataInterface() {propertyValues {FloatData [<8.3 -12.5 -1.5>, <14.25 -12 -2.25>, <10.5 -11 -2.2>, <0 0 0>]}}
      )
      False Vector=&SelectVectorFromArray(
       Direct Array Index=&SpawnGroup()
       Vector Selection Array=DataInterface() {propertyValues {FloatData [<8.3 -12.5 0>, <14.5 -12 0>, <10.5 -11.75 0>, <0 0 0>]}}
      )
     )
     Float=$System.MeshScaleRemap
    )
    Skeletal Mesh=$System.Skeletal Mesh
    Socket Sampling Mode=::NewEnumerator5
   ) {module=SkeletalMeshLocation001}
   &DynamicMaterialParameters(
    Index 0 Param 1=&Multiply_Float(A=&Multiply_Float(A=7.75, B=$System.MeshScaleAdjusted), B=1)
    Index 0 Param 2=&Multiply_Float(A=9, B=$System.MeshScaleAdjusted)
    Index 1 Param 1=&Multiply_Float(A=7, B=$System.MeshScaleAdjusted)
    Index 1 Param 2=&Multiply_Float(A=7, B=$System.MeshScaleAdjusted)
    Module.Index0Float0Write=true
    Module.Index0Float1Write=true
    Module.Index0Float2Write=true
    Module.Index0Float3Write=true
    Module.Index1Float0Write=true
    Module.Index1Float1Write=true
    Module.Index1Float2Write=true
    Module.Index1Float3Write=true
    Write Parameter Index 1=::NewEnumerator1
   )
   &SolveForcesAndVelocity(Clamp Velocity=false, Limit Acceleration=false)
   /SetVariables(Particles.SpriteAlignment=<0 0.2126 1>, Particles.SpriteFacing1:Vector3f=<0 1 0>) {isSetParameters=true, module=SetVariables_#4}
   &ScaleColor(
    Scale Alpha=&Multiply_Float(
     A=&MakeCustomFloatFromBool(
      Bool=&Boolean_OrOperation(Bool1=&Boolean_OrOperation(Bool1=$User.bIsFrontEnd, Bool2=$User.IsFrontEnd), Bool2=$User.bIsFrontEndPreview)
      True Float=0.7
      False Float=1
     )
     B=&FNTimeOfDayScalar(Night Scale=1.1, Morning Scale=1.15, Day Scale=1.15, Evening Scale=1.35)
    )
    Scale RGB=&Multiply_Vector(
     A=&VectorFromFloat(Value=1.75)
     B=&MakeVectorFromLinearColorRGB(Color=$System.TOD Color)
    )
    ScaleA=true
    ScaleRGB=true
   )
   /SetVariables(
    Particles.GlowArcColor:LinearColor=&MakeLinearColorFromVectorAndFloat(
     "Vector (RGB)"=&Multiply_Vector(
      A=&VectorFromFloat(Value=1)
      B=&MakeVectorFromLinearColorRGB(Color=$System.TOD Color)
     )
     "Float (Alpha)"=&Multiply_Float(
      A=&MakeCustomFloatFromBool(
       Bool=&Boolean_OrOperation(Bool1=&Boolean_OrOperation(Bool1=$User.bIsFrontEnd, Bool2=$User.IsFrontEnd), Bool2=$User.bIsFrontEndPreview)
       True Float=0.825
       False Float=0.975
      )
      B=&FNTimeOfDayScalar(Night Scale=1, Morning Scale=1, Evening Scale=1)
     )
    )
    Particles.HandleSplineColor:LinearColor=&MakeLinearColorFromVectorAndFloat(
     "Vector (RGB)"=&Multiply_Vector(
      A=&VectorFromFloat(Value=1.75)
      B=&MakeVectorFromLinearColorRGB(Color=$System.TOD Color)
     )
     "Float (Alpha)"=&Multiply_Float(
      A=&MakeCustomFloatFromBool(
       Bool=&Boolean_OrOperation(Bool1=&Boolean_OrOperation(Bool1=$User.bIsFrontEnd, Bool2=$User.IsFrontEnd), Bool2=$User.bIsFrontEndPreview)
       True Float=1.5
       False Float=3
      )
      B=&FNTimeOfDayScalar(Night Scale=1.15, Morning Scale=1.15, Evening Scale=1)
     )
    )
   ) {isSetParameters=true}
  ]
  rendererClasses [&NiagaraSpriteRendererProperties, &NiagaraMeshRendererProperties]
  renderers [
   {rendererClass=&NiagaraSpriteRendererProperties}
   {rendererClass=&NiagaraSpriteRendererProperties, rendererIndex=1}
   {rendererClass=&NiagaraMeshRendererProperties, rendererIndex=2}
   {rendererClass=&NiagaraMeshRendererProperties, rendererIndex=3}
  ]
 }
 StandFX_Combined {
  emitterUpdateScript [
   &EmitterState(
    Inactive Response=::NewEnumerator1
    Life Cycle Mode=::NewEnumerator1
    Loop Behavior=::NewEnumerator1
    Loop Duration Mode=::NewEnumerator1
    UseLoopDelay=false
   )
   &SpawnBurst_Instantaneous(Use Spawn Probability=false, UseLoopCountLimit=true)
   &SpawnBurst_Instantaneous(Spawn Group=1, Use Spawn Probability=false, UseLoopCountLimit=true) {module=SpawnBurst_Instantaneous001}
  ]
  particleSpawnScript [
   /SetVariables(Particles.VisibilityTag=&SpawnGroup()) {isSetParameters=true}
   &InitializeParticle(
    Color=&MaskLinearColorBySpawnGroup(Masked Linear Color=<rgba 1 1 1 0.8>, Passthrough Linear Color=<rgba 1 1 1 1>)
    Mesh Scale Mode=::NewEnumerator3
    Sprite Size=<35 32>
    Sprite Size Mode=::NewEnumerator4
    Sprite UV Mode=::NewEnumerator4
    UsePositionOffset=false
    Write Material Random=true
   )
   &CameraOffset(Camera Offset Amount=&MaskFloatBySpawnGroup(Masked Float=-7, Passthrough Float=-2))
   /SetVariables(Particles.SpriteSize2=<34.5 11.25>, Particles.UVScale2:Vector2f=<1 1>) {isSetParameters=true}
   &InitialMeshOrientation(Mesh Orientation Mode=::NewEnumerator2, Rotation=<0 0 0.25>, Use Rotation Vector=true)
  ]
  particleUpdateScript [
   &ParticleState(Kill Particles When Lifetime Has Elapsed=false, Let Infinitely Lived Particles Die When Emitter Deactivates=true)
   &SkeletalMeshLocation(
    Bone / Socket Index=&SelectIntFromArray(Direct Array Index=&SpawnGroup(), Int Selection Array=DataInterface() {propertyValues {IntData [2,3]}})
    MaskSpawnGroup=false
    Mesh Sampling Type=::NewEnumerator5
    Orientation Sampling=::NewEnumerator1
    Skeletal Mesh=$System.Skeletal Mesh
    Socket Sampling Mode=::NewEnumerator5
   )
   &ScaleSpriteSize(Scale Sprite Size Mode=::NewEnumerator0, Uniform Scale Factor=$System.MeshScaleRemap)
   &SolveForcesAndVelocity(Clamp Velocity=false, Limit Acceleration=false)
   &DynamicMaterialParameters(
    Index 0 Param 1=&Multiply_Float(A=&Multiply_Float(A=8, B=$System.MeshScaleAdjusted), B=1)
    Index 0 Param 2=&Multiply_Float(A=&Multiply_Float(A=6.5, B=$System.MeshScaleAdjusted), B=1)
    Module.Index0Float0Write=true
    Module.Index0Float1Write=true
    Module.Index0Float2Write=true
    Module.Index0Float3Write=true
   )
   /SetVariables(
    Particles.SpriteFacing=<0 0 1>
    Particles.SpritefFacing2:Vector3f=&ReturnCameraVector(Camera Vectors=::NewEnumerator1, Coordinate Space=::Simulation)
   ) {isSetParameters=true, module=SetVariables_#4}
   /SetVariables(
    Particles.SpriteAlignment=&ReturnCameraVector(Camera Vectors=::NewEnumerator1, Coordinate Space=::Simulation)
    Particles.SpriteAlignment2:Vector3f=<0 0 1>
   ) {isSetParameters=true}
   &ScaleColor(
    Scale Alpha=&MaskFloatBySpawnGroup(
     Masked Float=&Multiply_Float(
      A=&FN_FrontEndFloatModifier(Gameplay=1, Front End=1.5)
      B=&FNTimeOfDayScalar(Night Scale=1.1, Morning Scale=1.15, Day Scale=1.15, Evening Scale=1.35)
     )
     Passthrough Float=&Multiply_Float(
      A=&FN_FrontEndFloatModifier(Gameplay=1.4, Front End=1.12)
      B=&FNTimeOfDayScalar(Night Scale=1.15, Morning Scale=1.15, Evening Scale=1)
     )
    )
    Scale RGB=&MaskVectorBySpawnGroup(
     Masked Vector=&Multiply_Vector(
      A=&VectorFromFloat(Value=1.75)
      B=&MakeVectorFromLinearColorRGB(Color=$System.TOD Color)
     )
     Passthrough Vector=&Multiply_Vector(
      A=&VectorFromFloat(Value=1.75)
      B=&MakeVectorFromLinearColorRGB(Color=$System.TOD Color)
     )
    )
    ScaleA=true
    ScaleRGB=true
   )
   /SetVariables(Particles.SpriteSize2=&Multiply_Vector2D_ByFloat(Float=$System.MeshScaleRemap, Vector2D=$Particles.Initial.SpriteSize2)) {isSetParameters=true}
  ]
  rendererClasses [&NiagaraSpriteRendererProperties, &NiagaraMeshRendererProperties]
  renderers [
   {rendererClass=&NiagaraSpriteRendererProperties}
   {rendererClass=&NiagaraSpriteRendererProperties, rendererIndex=1}
   {rendererClass=&NiagaraMeshRendererProperties, rendererIndex=2}
  ]
 }
}
summary {
 systemName=NS_MicStand_PajamaSoar_Idle_FX
 emitters [
  {
   emitterName=StandFX_Combined
   rendererClasses [&NiagaraSpriteRendererProperties, &NiagaraMeshRendererProperties]
  }
  {emitterName=MicFX_Combined, rendererClasses [&NiagaraSpriteRendererProperties, &NiagaraMeshRendererProperties]}
 ]
 userVariables [
  User.bIsFrontEnd {type=NiagaraBool}
  User.bIsFrontEndPreview {type=NiagaraBool}
  User.IncludeMicFX {type=NiagaraFloat, defaultValue=1, description=""}
  User.IsBackpack {type=NiagaraFloat, defaultValue=0, description=""}
  User.IsFrontEnd {type=NiagaraBool}
 ]
}
